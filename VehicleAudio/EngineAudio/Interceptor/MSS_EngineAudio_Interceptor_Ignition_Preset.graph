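# Mono one-shot MetaSound for Interceptor vehicle engine ignition, layering a random ignition sample with delayed exhaust pops.
# An envelope stretches the interval between a random number of shuffled, pitch-randomized pops, while their gain eases down as the remaining count drops.
# OnFinished waits for both layers, treating an empty array as already finished.
# The default ExhaustPops array is empty here, so only the ignition plays.
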MetaSoundSource /VehicleAudio/EngineAudio/Interceptor/MSS_EngineAudio_Interceptor_Ignition_Preset
inputs [
 ExhaustPop_StartDelayTime {type=Time, access=Constructor}
 ExhaustPopPitch {type=Float}
 ExhaustPops {type=WaveAsset:Array, access=Constructor}
 ExhaustPopVolume {type=Float, access=Constructor}
 IgnitionSound {type=WaveAsset:Array, access=Constructor}
 IgnitionVolume {type=Float, access=Constructor}
 MaxNumExhaustPopsToPlay {type=Int32, access=Constructor}
 MaxRepeatRate {type=Float, access=Constructor}
 MinNumExhaustPopsToPlay {type=Int32, access=Constructor}
 MinRepeatRate {type=Float, access=Constructor}
 UE.Source.OnPlay {type=Trigger}
]
interfaces [UE.OutputFormat.Mono:1.0, UE.Source.OneShot:1.0, UE.Source:1.1]
outputs [
 "UE.OutputFormat.Mono.Audio:0" {type=Audio}
 UE.Source.OneShot.OnFinished {type=Trigger}
]
pages [Default {
  input_defaults {
   ExhaustPopPitch=0
   ExhaustPopVolume=-2
   ExhaustPop_StartDelayTime=0.8
   IgnitionVolume=2
   MaxNumExhaustPopsToPlay=10
   MaxRepeatRate=0.2
   MinNumExhaustPopsToPlay=4
   MinRepeatRate=0.1
   UE.Source.OnPlay=null
   ExhaustPops []
   IgnitionSound [&/EngineAudio_Interceptor_Ignition_01]
  }
  nodes [
   DecibelstoLinearGain_1 = UE.Decibels to Linear Gain:Float(Decibels=^IgnitionVolume)
   DecibelstoLinearGain_0 = UE.Decibels to Linear Gain:Float(Decibels=^ExhaustPopVolume)
   n1 = TriggerCompare.Clamp:Int32(Compare=^UE.Source.OnPlay, A=Array.Num:WaveAsset:Array(Array=^ExhaustPops).Num) {also {defaults {Type=3}}}
   TriggerDelay_0 = UE.Trigger Delay(Delay Time=^ExhaustPop_StartDelayTime, In=@n1.TRUE)
   ADSREnvelopeADSREnvelope_0 = ADSR Envelope.ADSR Envelope:Float(Trigger Attack=@TriggerDelay_0.Out) {also {defaults {Attack Time=2, Sustain Level=1}}}
   n2 = UE.RandomInt32(Max=^MaxNumExhaustPopsToPlay, Min=^MinNumExhaustPopsToPlay, Next=^UE.Source.OnPlay) {also {version [1,1]}}
   n3 = UE.TriggerRepeat(
    Period=UE.ConversionFloatToTime(In=MapRange.MapRange:Float(Out Range B=^MaxRepeatRate, Out Range A=^MinRepeatRate, In=@ADSREnvelopeADSREnvelope_0.Out Envelope).Out Value).Out
    Num Repeats=@n2.Value
    Start=@TriggerDelay_0.Out
   )
   TriggerCounter_0 = UE.Trigger Counter(In=@n3.RepeatOut) {also {defaults {Start Value=1}}}
   n4 = VariableAccessor.Int32(Variable=VariableMutator.Int32(Variable=InitVariable.Int32().Variable, Value=@n2.Value).Variable)
   n5 = Convert.Float:Int32(Int32=UE.Subtract:Int32(AdditionalOperands=@TriggerCounter_0.Count, PrimaryOperand=@n4.Value).Out)
   n6 = Convert.Float:Int32(Int32=VariableAccessor.Int32(Variable=@n4.Variable).Value)
   n7 = VariableAccessor.Float(Variable=VariableMutator.Float(Value=UE.Divide:Float(PrimaryOperand=@n5.Float, AdditionalOperands=@n6.Float).Out, Variable=InitVariable.Float().Variable).Variable)
   n8 = MapRange.MapRange:Float(In=@n7.Value) {also {defaults {Out Range A=0.25}}}
   n9 = Array.Shuffle:WaveAsset:Array(In Array=^ExhaustPops, Next=@n3.RepeatOut)
   n10 = UE.RandomFloat(Next=@n9.On Next) {also {defaults {Max=0.5, Min=-0.5}}}
   WavePlayer_0 = UE.Wave Player:Mono(Pitch Shift=UE.Add:Float(AdditionalOperands=^ExhaustPopPitch, PrimaryOperand=@n10.Value).Out, Wave Asset=@n9.Value, Play=@n10.On Next) {also {version [1,1]}}
   n11 = TriggerCompare.Clamp:Int32(Compare=^UE.Source.OnPlay, A=Array.Num:WaveAsset:Array(Array=^IgnitionSound).Num) {also {defaults {B=0, Type=1}}}
   ArrayRandomGet_0 = Array.Random Get:WaveAsset:Array(In Array=^IgnitionSound, Next=@n11.TRUE)
   WavePlayer_1 = UE.Wave Player:Mono(Play=@ArrayRandomGet_0.On Next, Wave Asset=@ArrayRandomGet_0.Value) {also {version [1,1]}}
   AudioMixerAudioMixerMono3_0 = "AudioMixer.Audio Mixer (Mono, 3)"(
    Gain 0=@DecibelstoLinearGain_1.Linear Gain
    Gain 1=UE.Multiply:Float(PrimaryOperand=@DecibelstoLinearGain_0.Linear Gain, AdditionalOperands=/VehicleAudio/Patches/EaseOutQuad(In=@n8.Out Value).Out).Out
    In 1=UE.SwitchAndRamp(Trigger=@n10.On Next, In Audio=@WavePlayer_0.Out Mono) {also {defaults {Smooth Time=0.05}}}.Out Audio
    In 0=@WavePlayer_1.Out Mono
   )
   InitVariable.Int32()
   TriggerAnyTriggerAny2_0 = "TriggerAny.Trigger Any (2)"(In 1=@n11.FALSE, In 0=@WavePlayer_1.On Finished)
   TriggerAnyTriggerAny2_1 = "TriggerAny.Trigger Any (2)"(
    In 1=TriggerCompare.Clamp:Float(A=VariableAccessor.Float(Variable=@n7.Variable).Value, Compare=@WavePlayer_0.On Finished).TRUE
    In 0=@n1.FALSE
   )
   TriggerAccumulatorTriggerAccumulate2_0 = "TriggerAccumulator.Trigger Accumulate (2)"(In 0=@TriggerAnyTriggerAny2_0.Out, In 1=@TriggerAnyTriggerAny2_1.Out)
  ]
  links [
   {dst="@UE.OutputFormat.Mono.Audio:0", src=AudioMixerAudioMixerMono3_0.Out}
   {dst="@UE.Source.OneShot.OnFinished", src=TriggerAccumulatorTriggerAccumulate2_0.Out}
  ]
 }]
properties {
 assetImportData=&self:AssetImportData
 attenuationSettings=/VehicleAudio/Settings/Vehicle_Movement_Att
 audioPropertiesSheet=/VehicleAudio/PropertySheets/APS_Cosmetic_Vehicle_Engine_Ignition
 maxDistance=12512
 sampleRate=48000
 soundClassObject=&Fort_Vehicles_Interior
 soundSubmixObject=&Fortnite_SFX
 concurrencySet [/VehicleAudio/Settings/Vehicle_Cosmetic_Ignition_Conc]
 modulationSettings {
  highpassRouting=Union
  lowpassRouting=Union
  volumeRouting=Union
  highpassModulationDestination {bEnableModulation=true, modulators [&PP_FortVehicles_Engine_HPF]}
  lowpassModulationDestination {bEnableModulation=true, modulators [&Patch_FortVehicles_Engine_LPF]}
  volumeModulationDestination {
   bEnableModulation=true
   modulators [&Patch_FortVehicles_Engine_Vol, &DM_VehicleEnginePatch]
  }
 }
}
