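# Looping low-spec mono MetaSound source for the Interceptor vehicle engine.
# Fort.Vehicle and REV inputs merge via max or trigger-any to drive a Rev Player, pitched by smoothed bonus speed plus vehicle pitch.
# Its left channel passes through throttle-based volume and filtering, then mixes with exhaust-pop and blowoff patches.
# The Default page modulates RPM and throttle through road-noise patches, which the Mobile page skips.
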MetaSoundSource /VehicleAudio/EngineAudio/Interceptor/MSS_EngineAudio_Interceptor_LowSpec_REV_Preset
inputs [
 Blowoff.Sound {type=WaveAsset:Array, access=Constructor}
 Blowoff.Volume {type=Float, access=Constructor}
 Character.Affiliation {type=Enum:FortAudioAffiliation}
 Character.Health {type=Float}
 Character.UniqueID {type=Int32}
 Downshift {type=Trigger}
 EngineBlowoff {type=Trigger}
 Exhaust.Pops.Blowoff.Volume {type=Float, access=Constructor}
 Exhaust.Pops.Long {type=WaveAsset:Array, access=Constructor}
 Exhaust.Pops.Pitch {type=Float, access=Constructor}
 Exhaust.Pops.Shift.Volume {type=Float, access=Constructor}
 Exhaust.Pops.Short {type=WaveAsset:Array, access=Constructor}
 Exhaust.Pops.Volume {type=Float}
 Exhaust.Throttle.EQ.FilterQ {type=Float, access=Constructor}
 Exhaust.Throttle.EQ.FilterType {type=Enum:BiquadFilterType, access=Constructor}
 Exhaust.Throttle.EQ.OffFreq {type=Float, access=Constructor}
 Exhaust.Throttle.EQ.OffGain {type=Float, access=Constructor}
 Exhaust.Throttle.EQ.OnFreq {type=Float, access=Constructor}
 Exhaust.Throttle.EQ.OnGain {type=Float, access=Constructor}
 Exhaust.Throttle.InterpTime {type=Time, access=Constructor}
 Exhaust.Throttle.Volume.Off {type=Float, access=Constructor}
 Exhaust.Throttle.Volume.On {type=Float, access=Constructor}
 Fort.Vehicle.Acceleration {type=Float}
 Fort.Vehicle.BlowoffStrength {type=Float}
 Fort.Vehicle.BonusSpeed {type=Float}
 Fort.Vehicle.EnableShifting {type=Bool}
 Fort.Vehicle.Gear {type=Int32}
 Fort.Vehicle.IsGrounded {type=Bool}
 Fort.Vehicle.Load {type=Float}
 Fort.Vehicle.NormalSpeed {type=Float}
 Fort.Vehicle.OnBlowoff {type=Trigger}
 Fort.Vehicle.OnDownshift {type=Trigger}
 Fort.Vehicle.OnThrottleEngaged {type=Trigger}
 Fort.Vehicle.OnThrottleReleased {type=Trigger}
 Fort.Vehicle.OnUpshift {type=Trigger}
 Fort.Vehicle.Pitch {type=Float}
 Fort.Vehicle.RPM {type=Float}
 Fort.Vehicle.Throttle {type=Float}
 GearChangeTime {type=Time, access=Constructor}
 IsLocal {type=Bool}
 Mono {type=Rev Asset, access=Constructor}
 REV.EnableShifting {type=Bool}
 REV.Gear {type=Int32}
 REV.RPM {type=Float}
 REV.Throttle {type=Float}
 REV.Velocity {type=Float}
 REVVolume {type=Float}
 Settings {type=Rev Settings, access=Constructor}
 Speed_Normalized {type=Float}
 UE.Source.OnPlay {type=Trigger}
 UE.Source.Orientation.Azimuth {type=Float}
 UE.Source.Orientation.Elevation {type=Float}
 Upshift {type=Trigger}
]
interfaces [Fort.Character:1.0, Fort.Vehicle:1.0, REV:1.0, UE.OutputFormat.Mono:1.0, UE.Source.Orientation:1.0, UE.Source:1.1]
outputs [
 Downshift {type=Trigger}
 Fundamental Frequency {type=Float}
 Gear {type=Int32}
 RPM {type=Float}
 SpeedNormalized {type=Float}
 Throttle {type=Float}
 "UE.OutputFormat.Mono.Audio:0" {type=Audio}
 Upshift {type=Trigger}
]
pages [
 Default {
  is_build_graph=true
  input_defaults {
   Blowoff.Volume=0
   Character.Affiliation=0
   Character.Health=0
   Character.UniqueID=0
   Downshift=null
   EngineBlowoff=null
   Exhaust.Pops.Blowoff.Volume=0
   Exhaust.Pops.Pitch=0
   Exhaust.Pops.Shift.Volume=0
   Exhaust.Pops.Volume=0
   Exhaust.Throttle.EQ.FilterQ=1
   Exhaust.Throttle.EQ.FilterType=4
   Exhaust.Throttle.EQ.OffFreq=20000
   Exhaust.Throttle.EQ.OffGain=0
   Exhaust.Throttle.EQ.OnFreq=20000
   Exhaust.Throttle.EQ.OnGain=0
   Exhaust.Throttle.InterpTime=0.1
   Exhaust.Throttle.Volume.Off=0
   Exhaust.Throttle.Volume.On=0
   Fort.Vehicle.Acceleration=0
   Fort.Vehicle.BlowoffStrength=0
   Fort.Vehicle.BonusSpeed=0
   Fort.Vehicle.EnableShifting=null
   Fort.Vehicle.Gear=0
   Fort.Vehicle.IsGrounded="<MetaSoundBuilderResult.SUCCEEDED: 0>"
   Fort.Vehicle.Load=0
   Fort.Vehicle.NormalSpeed=0
   Fort.Vehicle.OnBlowoff=null
   Fort.Vehicle.OnDownshift=null
   Fort.Vehicle.OnThrottleEngaged=null
   Fort.Vehicle.OnThrottleReleased=null
   Fort.Vehicle.OnUpshift=null
   Fort.Vehicle.Pitch=0
   Fort.Vehicle.RPM=0
   Fort.Vehicle.Throttle=0
   GearChangeTime=0.1
   IsLocal=null
   Mono=&/Interceptor
   REV.EnableShifting=null
   REV.Gear=0
   REV.RPM=0
   REV.Throttle=0
   REV.Velocity=0
   REVVolume=7
   Settings=&/REVSettings_Interceptor
   Speed_Normalized=0
   UE.Source.OnPlay=null
   UE.Source.Orientation.Azimuth=0
   UE.Source.Orientation.Elevation=0
   Upshift=null
   Blowoff.Sound []
   Exhaust.Pops.Long []
   Exhaust.Pops.Short []
  }
  nodes [
   TriggerAnyTriggerAny2_1 = "TriggerAny.Trigger Any (2)"(In 1=^EngineBlowoff, In 0=^Fort.Vehicle.OnBlowoff)
   n1 = VariableAccessor.Trigger(Variable=VariableMutator.Trigger(Variable=InitVariable.Trigger().Variable, Value=@TriggerAnyTriggerAny2_1.Out).Variable)
   TriggerAnyTriggerAny2_3 = "TriggerAny.Trigger Any (2)"(In 0=@n1.Value)
   n2 = VariableDeferredAccessor.Float(Variable=InitVariable.Float().Variable)
   TriggerAnyTriggerAny2_4 = "TriggerAny.Trigger Any (2)"(In 1=^Fort.Vehicle.OnUpshift, In 0=^Upshift)
   n3 = VariableMutator.Trigger(Variable=InitVariable.Trigger().Variable, Value=@TriggerAnyTriggerAny2_4.Out)
   TriggerAnyTriggerAny2_0 = "TriggerAny.Trigger Any (2)"(In 1=^Downshift, In 0=^Fort.Vehicle.OnDownshift)
   n4 = VariableMutator.Trigger(Variable=InitVariable.Trigger().Variable, Value=@TriggerAnyTriggerAny2_0.Out)
   n5 = MapRange.MapRange:Float(In=^Fort.Vehicle.BonusSpeed) {also {defaults {Out Range B=0.75}}}
   VariableAccessorFloat_1 = VariableAccessor.Float(Variable=VariableMutator.Float(Variable=InitVariable.Float().Variable, Value=Max.Max:Float(B=^Fort.Vehicle.RPM, A=^REV.RPM).Value).Variable)
   DM_EngineAudio_RoadNoiseParamModifier_0 = /VehicleAudio/Patches/DM_EngineAudio_RoadNoiseParamModifier(Input=@VariableAccessorFloat_1.Value)
   n6 = VariableDeferredAccessor.Float(Variable=@n2.Variable)
   VariableAccessorFloat_0 = VariableAccessor.Float(Variable=VariableMutator.Float(Value=Max.Max:Float(B=^Fort.Vehicle.Throttle, A=^REV.Throttle).Value, Variable=@n6.Variable).Variable)
   DM_EngineAudio_RoadNoiseParamModifier_1 = &DM_EngineAudio_RoadNoiseParamModifier(Input=@VariableAccessorFloat_0.Value)
   n7 = InitVariable.Int32()
   MaxMax_3 = Max.Max:Int32(B=^Fort.Vehicle.Gear, A=^REV.Gear)
   REVRevPlayer_0 = REV.Rev Player:Stereo(
    Rev Asset=^Mono
    Enable Shifting=^REV.EnableShifting
    Rev Settings=^Settings
    Play=^UE.Source.OnPlay
    Pitch Shift=UE.Add:Float(AdditionalOperands=^Fort.Vehicle.Pitch, PrimaryOperand=UE.InterpTo:Audio(Target=@n5.Out Value) {also {defaults {Interp Time=0.15}}}.Value).Out
    RPM=@DM_EngineAudio_RoadNoiseParamModifier_0.Out
    Throttle=@DM_EngineAudio_RoadNoiseParamModifier_1.Out
    Gear=VariableAccessor.Int32(Variable=VariableMutator.Int32(Variable=@n7.Variable, Value=@MaxMax_3.Value).Variable).Value
   )
   DecibelstoLinearGain_0 = UE.Decibels to Linear Gain:Float(Decibels=^REVVolume)
   DecibelstoLinearGain_1 = UE.Decibels to Linear Gain:Float(Decibels=^Exhaust.Pops.Volume)
   AudioMixerAudioMixerMono3_0 = "AudioMixer.Audio Mixer (Mono, 3)"(
    In 2=/VehicleAudio/Patches/DM_EngineAudio_Blowoff(
     BlowoffSound=^Blowoff.Sound
     OutputVolume=^Blowoff.Volume
     Character.Affiliation=^Character.Affiliation
     IsLocal=^IsLocal
     UE.Source.OnPlay=^UE.Source.OnPlay
     EngineBlowoff=@TriggerAnyTriggerAny2_3.Out
     REV.Throttle=@n2.Value
    )."UE.OutputFormat.Mono.Audio:0"
    In 0=/VehicleAudio/Patches/DM_EngineAudio_ExhaustPops(
     Character.Affiliation=^Character.Affiliation
     BlowoffVolume=^Exhaust.Pops.Blowoff.Volume
     ExhaustPops_Long=^Exhaust.Pops.Long
     Pitch=^Exhaust.Pops.Pitch
     ShiftVolume=^Exhaust.Pops.Shift.Volume
     ExhaustPops_Short=^Exhaust.Pops.Short
     GearChangeTime=^GearChangeTime
     IsLocal=^IsLocal
     Upshift=VariableAccessor.Trigger(Variable=@n3.Variable).Value
     Downshift=VariableAccessor.Trigger(Variable=@n4.Variable).Value
     Blowoff=VariableAccessor.Trigger(Variable=@n1.Variable).Value
    )."UE.OutputFormat.Mono.Audio:0"
    In 1=/VehicleAudio/Patches/DM_EngineAudio_ThrottleVolumeFilter(
     Filter.FilterQ=^Exhaust.Throttle.EQ.FilterQ
     Filter.FilterType=^Exhaust.Throttle.EQ.FilterType
     Filter.OffThrottleFreq=^Exhaust.Throttle.EQ.OffFreq
     Filter.OffThrottleGain=^Exhaust.Throttle.EQ.OffGain
     Filter.OnThrottleFreq=^Exhaust.Throttle.EQ.OnFreq
     Filter.OnThrottleGain=^Exhaust.Throttle.EQ.OnGain
     Interp Time=^Exhaust.Throttle.InterpTime
     Volume.OffThrottle=^Exhaust.Throttle.Volume.Off
     Volume.OnThrottle=^Exhaust.Throttle.Volume.On
     AudioIn=@REVRevPlayer_0.Out Left
     REV.Throttle=@n6.Value
    ).AudioOut
    Gain 1=@DecibelstoLinearGain_0.Linear Gain
    Gain 0=@DecibelstoLinearGain_1.Linear Gain
   )
   MaxMax_2 = Max.Max:Float(B=^Fort.Vehicle.NormalSpeed, A=^Speed_Normalized)
   VariableMutator.Float(Variable=InitVariable.Float().Variable, Value=@MaxMax_2.Value)
   TriggerAnyTriggerAny2_2 = "TriggerAny.Trigger Any (2)"(In 0=^Downshift, In 1=^Upshift)
   VariableMutator.Trigger(Variable=InitVariable.Trigger().Variable, Value=@TriggerAnyTriggerAny2_2.Out)
  ]
  links [
   {dst="@UE.OutputFormat.Mono.Audio:0", src=AudioMixerAudioMixerMono3_0.Out}
   {dst="@RPM", src=DM_EngineAudio_RoadNoiseParamModifier_0.Out}
   {dst="@Throttle", src=DM_EngineAudio_RoadNoiseParamModifier_1.Out}
   {dst="@SpeedNormalized", src=MaxMax_2.Value}
   {dst="@Gear", src=MaxMax_3.Value}
   {dst="@Fundamental Frequency", src=REVRevPlayer_0.Fundamental Frequency}
   {dst="@Downshift", src=TriggerAnyTriggerAny2_0.Out}
   {dst="@Upshift", src=TriggerAnyTriggerAny2_4.Out}
  ]
 }
 Mobile {
  nodes [
   TriggerAnyTriggerAny2_1 = "TriggerAny.Trigger Any (2)"(In 1=^EngineBlowoff, In 0=^Fort.Vehicle.OnBlowoff)
   n1 = VariableAccessor.Trigger(Variable=VariableMutator.Trigger(Variable=InitVariable.Trigger().Variable, Value=@TriggerAnyTriggerAny2_1.Out).Variable)
   TriggerAnyTriggerAny2_3 = "TriggerAny.Trigger Any (2)"(In 0=@n1.Value)
   n2 = VariableDeferredAccessor.Float(Variable=InitVariable.Float().Variable)
   TriggerAnyTriggerAny2_4 = "TriggerAny.Trigger Any (2)"(In 1=^Fort.Vehicle.OnUpshift, In 0=^Upshift)
   n3 = VariableAccessor.Trigger(Variable=VariableMutator.Trigger(Variable=InitVariable.Trigger().Variable, Value=@TriggerAnyTriggerAny2_4.Out).Variable)
   TriggerAnyTriggerAny2_0 = "TriggerAny.Trigger Any (2)"(In 1=^Downshift, In 0=^Fort.Vehicle.OnDownshift)
   n4 = VariableMutator.Trigger(Variable=InitVariable.Trigger().Variable, Value=@TriggerAnyTriggerAny2_0.Out)
   n5 = MapRange.MapRange:Float(In=^Fort.Vehicle.BonusSpeed) {also {defaults {Out Range B=0.75}}}
   n6 = VariableDeferredAccessor.Float(Variable=@n2.Variable)
   VariableAccessorFloat_0 = VariableAccessor.Float(Variable=VariableMutator.Float(Value=Max.Max:Float(B=^Fort.Vehicle.Throttle, A=^REV.Throttle).Value, Variable=@n6.Variable).Variable)
   VariableAccessorFloat_1 = VariableAccessor.Float(Variable=VariableMutator.Float(Variable=InitVariable.Float().Variable, Value=Max.Max:Float(B=^Fort.Vehicle.RPM, A=^REV.RPM).Value).Variable)
   n7 = InitVariable.Int32()
   MaxMax_3 = Max.Max:Int32(B=^Fort.Vehicle.Gear, A=^REV.Gear)
   REVRevPlayer_0 = REV.Rev Player:Stereo(
    Rev Asset=^Mono
    Enable Shifting=^REV.EnableShifting
    Rev Settings=^Settings
    Play=^UE.Source.OnPlay
    Pitch Shift=UE.Add:Float(AdditionalOperands=^Fort.Vehicle.Pitch, PrimaryOperand=UE.InterpTo:Audio(Target=@n5.Out Value) {also {defaults {Interp Time=0.15}}}.Value).Out
    Throttle=@VariableAccessorFloat_0.Value
    RPM=@VariableAccessorFloat_1.Value
    Gear=VariableAccessor.Int32(Variable=VariableMutator.Int32(Variable=@n7.Variable, Value=@MaxMax_3.Value).Variable).Value
   )
   DecibelstoLinearGain_0 = UE.Decibels to Linear Gain:Float(Decibels=^REVVolume)
   DecibelstoLinearGain_1 = UE.Decibels to Linear Gain:Float(Decibels=^Exhaust.Pops.Volume)
   AudioMixerAudioMixerMono3_0 = "AudioMixer.Audio Mixer (Mono, 3)"(
    In 2=&DM_EngineAudio_Blowoff(
     BlowoffSound=^Blowoff.Sound
     OutputVolume=^Blowoff.Volume
     Character.Affiliation=^Character.Affiliation
     IsLocal=^IsLocal
     UE.Source.OnPlay=^UE.Source.OnPlay
     EngineBlowoff=@TriggerAnyTriggerAny2_3.Out
     REV.Throttle=@n2.Value
    )."UE.OutputFormat.Mono.Audio:0"
    In 0=&DM_EngineAudio_ExhaustPops(
     Character.Affiliation=^Character.Affiliation
     BlowoffVolume=^Exhaust.Pops.Blowoff.Volume
     ExhaustPops_Long=^Exhaust.Pops.Long
     Pitch=^Exhaust.Pops.Pitch
     ShiftVolume=^Exhaust.Pops.Shift.Volume
     ExhaustPops_Short=^Exhaust.Pops.Short
     GearChangeTime=^GearChangeTime
     IsLocal=^IsLocal
     Upshift=@n3.Value
     Downshift=VariableAccessor.Trigger(Variable=@n4.Variable).Value
     Blowoff=VariableAccessor.Trigger(Variable=@n1.Variable).Value
    )."UE.OutputFormat.Mono.Audio:0"
    In 1=&DM_EngineAudio_ThrottleVolumeFilter(
     Filter.FilterQ=^Exhaust.Throttle.EQ.FilterQ
     Filter.FilterType=^Exhaust.Throttle.EQ.FilterType
     Filter.OffThrottleFreq=^Exhaust.Throttle.EQ.OffFreq
     Filter.OffThrottleGain=^Exhaust.Throttle.EQ.OffGain
     Filter.OnThrottleFreq=^Exhaust.Throttle.EQ.OnFreq
     Filter.OnThrottleGain=^Exhaust.Throttle.EQ.OnGain
     Interp Time=^Exhaust.Throttle.InterpTime
     Volume.OffThrottle=^Exhaust.Throttle.Volume.Off
     Volume.OnThrottle=^Exhaust.Throttle.Volume.On
     AudioIn=@REVRevPlayer_0.Out Left
     REV.Throttle=@n6.Value
    ).AudioOut
    Gain 1=@DecibelstoLinearGain_0.Linear Gain
    Gain 0=@DecibelstoLinearGain_1.Linear Gain
   )
   PrintLogPrintLog_0 = Print Log.Print Log:String() {also {defaults {Label=Throttle, Value To Log=Released}}}
   PrintLogPrintLog_1 = Print Log.Print Log:String() {also {defaults {Label=Throttle, Value To Log=Engaged}}}
   VariableAccessor.Trigger(Variable=@n3.Variable)
   MaxMax_2 = Max.Max:Float(B=^Fort.Vehicle.NormalSpeed, A=^Speed_Normalized)
   VariableMutator.Float(Variable=InitVariable.Float().Variable, Value=@MaxMax_2.Value)
   TriggerAnyTriggerAny2_2 = "TriggerAny.Trigger Any (2)"(In 0=^Downshift, In 1=^Upshift)
   VariableMutator.Trigger(Variable=InitVariable.Trigger().Variable, Value=@TriggerAnyTriggerAny2_2.Out)
  ]
  links [
   {dst="@UE.OutputFormat.Mono.Audio:0", src=AudioMixerAudioMixerMono3_0.Out}
   {dst="@SpeedNormalized", src=MaxMax_2.Value}
   {dst="@Gear", src=MaxMax_3.Value}
   {dst="@Fundamental Frequency", src=REVRevPlayer_0.Fundamental Frequency}
   {dst="@Downshift", src=TriggerAnyTriggerAny2_0.Out}
   {dst="@Upshift", src=TriggerAnyTriggerAny2_4.Out}
   {dst="@Throttle", src=VariableAccessorFloat_0.Value}
   {dst="@RPM", src=VariableAccessorFloat_1.Value}
  ]
 }
]
properties {
 assetImportData=&self:AssetImportData
 attenuationSettings=&Vehicle_Engine_Att
 audioPropertiesSheet=&APS_Cosmetic_Vehicle_Engine
 bLooping=true
 duration=10000
 maxDistance=12000
 priority=80
 qualitySetting=Vehicles_REV_Mono
 sampleRate=48000
 soundClassObject=&Fort_Vehicles_Engine
 soundSubmixObject=&Fortnite_SFX
 volume=0.7143
 concurrencySet [&Conc_Veh_EngineAudio_LowSpec, &Conc_Veh_EngineAudio_LowSpec_Actor]
 modulationSettings {
  highpassRouting=Union
  lowpassRouting=Union
  pitchRouting=Union
  volumeRouting=Union
  highpassModulationDestination {bEnableModulation=true, modulators [&PP_FortVehicles_Engine_HPF]}
  lowpassModulationDestination {bEnableModulation=true, modulators [&Patch_FortVehicles_Engine_LPF]}
  pitchModulationDestination {bEnableModulation=true, modulators [&DM_Bus_Vehicle_Engine_Pitch]}
  volumeModulationDestination {
   bEnableModulation=true
   modulators [&DM_VehicleEnginePatch, &Patch_FortVehicles_Engine_Vol]
  }
 }
}
